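# Physics asset of collision shapes for the WealthLamb_Fate male medium character body, evidently for its simulated cape to collide against.
# Capsules cover the pelvis, spine and clavicles, with overlapping capsules on the upper three spine bones and a single sphere on the head.
# Arm, leg and foot bodies exist but hold no shapes.
# It has no constraints, so it defines collision volumes rather than a jointed ragdoll.
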PhysicsAsset /BRCosmetics/Characters/Player/Male/Medium/Bodies/M_MED_WealthLamb_Fate/Meshes/M_MED_WealthLamb_Fate_CapeColliders
bodies {
 calf_l []
 calf_r []
 clavicle_l [{shapeType=Capsule, length=10.2118, radius=8.6397, center=<8.135 -0.7451 0.7697>, rotation=<rot 74.0536 0.5302 -178.4158>}]
 clavicle_r [{shapeType=Capsule, length=10.4768, radius=8.2107, center=<-7.2949 0.7288 -1.0097>, rotation=<rot 74.0536 0.5302 1.5842>}]
 foot_l []
 foot_r []
 head [{shapeType=Sphere, radius=17.3125, center=<11.028 9.0474 0>}]
 lowerarm_l []
 lowerarm_r []
 pelvis [{shapeType=Capsule, length=13.5067, radius=14.5388, center=<-0.8675 0.1529 0>}]
 spine_01 [{shapeType=Capsule, length=9.3659, radius=13.5586, center=<-1.0083 0.9486 0>}]
 spine_02 [{shapeType=Capsule, length=8.8464, radius=13.6322, center=<-1.711 1.8659 0.0001>}]
 spine_03 [
  {shapeType=Capsule, length=11.1725, radius=13.7031, center=<0.275 2.8734 0.0001>}
  {shapeType=Capsule, length=12.4394, radius=13.7031, shapeName=spine_03_capsule, center=<6.0683 1.9926 0.0001>}
 ]
 spine_04 [
  {shapeType=Capsule, length=16.3492, radius=14.1325, center=<2.766 1.7502 0.0002>}
  {shapeType=Capsule, length=16.9, radius=14.1325, shapeName=spine_04_capsule, center=<5.3365 1.2745 0.0002>}
 ]
 spine_05 [
  {shapeType=Capsule, length=20.0594, radius=10.9469, center=<-2.8535 -1.7919 0>}
  {shapeType=Capsule, length=20.4209, radius=10.7353, center=<1.1621 -1.5171 0>}
  {shapeType=Capsule, length=9.7462, radius=9.4011, shapeName=spine_05_capsule, center=<4.2249 -1.0464 0>}
 ]
 thigh_l []
 thigh_r []
 upperarm_l []
 upperarm_r []
}
constraints []
